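# Niagara system for the ChillCat emote's laser, a red, elongated mesh particle.
# A single burst emitter places the mesh on the user-supplied skeletal mesh and orients it toward an offset point rotated with the owner.
# A local GetCameraFacing system module uses the camera interface, and a system forward vector sets the mesh's up direction.
# User.Alpha scales opacity, defaulting to zero, while randomized and age-lerped dynamic parameters drive the material.
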NiagaraSystem /BRCosmetics/Effects/Fort_Effects/Emotes/ChillCat/NS_Emote_ChillCat_Laser
dependencies {
 usedDataInterfaces [&NiagaraDataInterfaceCamera]
 usedDynamicInputs []
 usedModules [&SystemState, &self:GetCameraFacing]
}
emitters {Laser {
  emitterUpdateScript [
   &EmitterState()
   &SpawnBurst_Instantaneous(Use Spawn Probability=false, UseLoopCountLimit=true)
  ]
  particleSpawnScript [&InitializeParticle(Color=<rgba 10 0 0 1>, Mesh Scale=<0.3 20 1>, Mesh Scale Mode=::NewEnumerator4, UsePositionOffset=false, Write Material Random=true)]
  particleUpdateScript [
   &ParticleState(Kill Particles When Lifetime Has Elapsed=false, Let Infinitely Lived Particles Die When Emitter Deactivates=true)
   &DynamicMaterialParameters(
    Index 0 Param 1=0.8
    Index 0 Param 2=&RandomRangeFloat(Minimum=0, Maximum=1)
    Index 0 Param 3=&Lerp_Float(
     A=&RandomRangeFloat(Minimum=0, Maximum=1)
     B=&RandomRangeFloat(Minimum=-1, Maximum=1)
     Alpha=$Particles.NormalizedAge
    )
    Index 0 Param 4=1
    Module.Param0WriteEnabled=true
    Module.Param1WriteEnabled=true
    Module.Param2WriteEnabled=true
    Module.Param3WriteEnabled=true
   )
   &SkeletalMeshLocation(Bone Sampling Mode=::NewEnumerator4, MaskSpawnGroup=false, Orientation Sampling=::NewEnumerator1, Skeletal Mesh=$User.SkeletalMesh)
   &ScaleColor(Scale Alpha=$User.Alpha, ScaleA=true, ScaleRGB=false)
   &UpdateMeshOrientation(
    Facing Interpolation=false
    Facing Position:NiagaraPosition=&AddVectorToPosition(
     Position=$Engine.Owner.Position
     Vector=&RotateVectorByQuaternion(Vector To Rotate=<0 -25 240>, Quaternion=$Engine.Owner.Rotation)
    )
    Orientation Method=::NewEnumerator2
    Override Quat=false
    TransformFacingOrigin=false
    TransformFacingPosition=false
    TransformUp=false
    Up Direction:Vector3f=$System.Forward Vector World
   )
  ]
  rendererClasses [&NiagaraMeshRendererProperties]
  renderers [{rendererClass=&NiagaraMeshRendererProperties, rendererIndex=0}]
 }}
summary {
 systemName=NS_Emote_ChillCat_Laser
 emitters [{emitterName=Laser, rendererClasses [&NiagaraMeshRendererProperties]}]
 userVariables [
  User.Alpha {type=NiagaraFloat, defaultValue=0, description=""}
  User.bIsFrontEnd {type=NiagaraBool}
  User.bIsFrontEndPreview {type=NiagaraBool}
  User.SkeletalMesh {
   type=NiagaraDataInterfaceSkeletalMesh
   defaultValue {
    struct=NiagaraExt_VariableValue_DataInterface
    value {dataInterface=&self:NiagaraDataInterfaceSkeletalMesh_1}
   }
  }
 ]
}
